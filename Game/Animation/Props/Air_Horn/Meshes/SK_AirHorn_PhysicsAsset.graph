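# Physics asset for the air horn prop's skeletal mesh, SK_AirHorn, giving it a simple collision shape.
# A single body on the root bone holds one capsule, offset and rotated from the bone origin.
# It has no constraints, so the prop is represented as one piece with no jointed parts.
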PhysicsAsset /Game/Animation/Props/Air_Horn/Meshes/SK_AirHorn_PhysicsAsset
bodies {root [{shapeType=Capsule, length=13.9365, radius=7.9412, center=<0.0638 2.7766 1.7126>, rotation=<rot 7.0859 -90 -0.4958>}]}
constraints []
